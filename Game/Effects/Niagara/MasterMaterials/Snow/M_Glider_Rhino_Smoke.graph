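# Unlit translucent sprite material for smoke, named for the Rhino glider, usable on Niagara and particle sprites.
# A desaturated SubUV smoke-puff texture, scaled by particle color and time-of-day brightness, feeds an emissive-adjustment function, with optional time-of-day tinting.
# Opacity multiplies particle alpha by the texture's red channel, with depth fade skipped on mobile feature levels.
# Static switches add optional time-of-day alpha and a soft sphere mask around the object position.
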Material /Game/Effects/Niagara/MasterMaterials/Snow/M_Glider_Rhino_Smoke
graph {
 properties {
  bUsedWithNiagaraSprites=true
  bUsedWithParticleSprites=true
  blendMode=BLEND_Translucent
  pixelDepthOffsetMode=PDOM_Legacy
  shadingModel=MSM_Unlit
  usedShadingModels=Unlit
 }
 nodes [
  n1 = ScalarParameter() {defaultValue=1, parameterName=Day color}
  n2 = ScalarParameter() {defaultValue=0.5, parameterName=Dark color}
  n3 = TextureSampleParameterSubUV!(UVs, Apply View MipBias) {
   group=Normal Map
   parameterName=Normal Map SubUV
   samplerType=SAMPLERTYPE_LinearColor
   texture=/Game/Effects/Fort_Effects/Gameplay/SmokeDustExplosions/Textures/T_SmokePoof_8x8_01
  }
  n4 = ParticleColor()
  n5 = Multiply(
   A=Multiply(
    A=MaterialFunctionCall(Color_Morning=@n1, Color_Day=@n1, Color_Evening=@n2, Color_Night=@n2) {materialFunction=&Color_Time_of_Day}.Result
    B=Desaturation(None=@n3.RGB)
   )
   B=@n4.RGB
  )
  n6 = Multiply(A=@n5, B=VectorParameter() {parameterName=Smoke_color_evening, defaultValue=<rgba 0.651 0.4525 0.4331 1>}.RGB)
  n7 = Multiply(A=@n5, B=VectorParameter() {parameterName=Smoke_color_night, defaultValue=<rgba 0.1716 0.2135 0.349 1>}.RGB)
  MaterialFunctionCall(Emissive=StaticSwitchParameter(
    True=MaterialFunctionCall(Color_Morning=@n6, Color_Day=@n5, Color_Evening=@n6, Color_Night=@n7) {materialFunction=&Color_Time_of_Day}.Result
    False=@n5
   ) {parameterName=Use_smoke_color_time_of_day}) {materialFunction=&MF_FX_EmissiveAdjustments}
  n8 = ScalarParameter() {defaultValue=1, parameterName=Smoke_alpha_evening}
  n9 = ScalarParameter() {defaultValue=0.5, parameterName=Smoke_alpha_day}
  n10 = ScalarParameter() {defaultValue=1, parameterName=Smoke_alpha_night}
  n11 = Multiply(A=@n4.A, B=@n3.R)
  n12 = DepthFade(Opacity=@n11, FadeDistance=ScalarParameter() {defaultValue=30, parameterName=Depth_Fade})
  n13 = StaticSwitchParameter(True=FeatureLevelSwitch(Default=@n12, ES2=@n11, ES3_1=@n11, SM4_REMOVED=@n12, SM5=@n12, SM6=@n12), False=@n11) {defaultValue=true, parameterName=Use depth}
  n14 = StaticSwitchParameter(
   True=Multiply(
    A=MaterialFunctionCall(Color_Morning=@n8, Color_Day=@n9, Color_Evening=@n8, Color_Night=@n10) {materialFunction=&Color_Time_of_Day}.Result
    B=@n13
   )
   False=@n13
  ) {parameterName=Use_smoke_alpha_time_of_day}
  n15 = WorldPosition()
  n16 = ObjectPositionWS()
  n17 = Constant() {r=125}
  StaticSwitchParameter(True=Multiply(A=@n14, B=SphereMask(A=@n15.XYZ, B=@n16, Radius=@n17) {hardnessPercent=0}), False=@n14) {parameterName=Use world position mask}
 ]
}
