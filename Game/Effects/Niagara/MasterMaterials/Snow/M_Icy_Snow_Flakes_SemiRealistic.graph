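# Translucent unlit, two-sided particle material for icy, semi-realistic snowflake sprites, tinted by particle color through an emissive adjustment function.
# Panning noise distorts UVs for a motion-blurred snowflake texture, which is squared and multiplied by swirl and turbulence masks to form the shape.
# Opacity combines a random per-particle channel pick, particle alpha, near-camera fade and optional distance cull fade.
# World position offset collapses sprites near the camera, with shading path switches disabling this and the camera fade on mobile.
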Material /Game/Effects/Niagara/MasterMaterials/Snow/M_Icy_Snow_Flakes_SemiRealistic
graph {
 properties {
  bEnableResponsiveAA=true
  bUsedWithNiagaraSprites=true
  bUsedWithParticleSprites=true
  blendMode=BLEND_Translucent
  pixelDepthOffsetMode=PDOM_Legacy
  shadingModel=MSM_Unlit
  translucencyLightingMode=TLM_VolumetricPerVertexDirectional
  twoSided=true
  usedShadingModels=Unlit
 }
 nodes [
  n1 = ParticleColor()
  MaterialFunctionCall(Emissive=@n1.RGB) {materialFunction=&MF_FX_EmissiveAdjustments}
  n2 = TextureCoordinate()
  n3 = Constant2Vector() {g=1, r=1}
  n4 = TextureSample(UVs=Panner(Coordinate=Multiply(A=@n2, B=@n3)) {speedX=0.2, speedY=0.1}) {samplerType=SAMPLERTYPE_LinearColor, texture=/Game/Effects/Fort_Effects/Gameplay/SmokeDustExplosions/Textures/T_NoisePack_LinearColor}
  n5 = Add(A=TextureCoordinate(), B=Multiply(A=ComponentMask(None=@n4.RGB) {g=true, r=true}) {constB=0.025})
  n6 = MaterialFunctionCall!(
   TextureObject=TextureObject() {texture=/Game/Effects/Fort_Effects/Textures/Snowflake/T_Snowflakes_F}
   UV=@n5
   Vector=Constant2Vector() {g=-3, r=3}
   Distance
   Steps
  ) {materialFunction=/Engine/Functions/Engine_MaterialFunctions01/Texturing/MotionBlur-Texture}
  n7 = Multiply(A=@n6.Result, B=@n6.Result)
  n8 = TextureSample(UVs=Multiply(A=@n5, B=Constant2Vector() {g=1, r=0.935})) {
   samplerType=SAMPLERTYPE_Masks
   texture=/Game/Creative/Sets/PropSets/Primitives/Customization/Textures/GalacticSwirl/T_CM_GalacticSwirl_Flipbook
  }
  n9 = Multiply(A=@n8.G) {constB=15}
  n10 = Constant2Vector() {g=1, r=1}
  n11 = Multiply(A=TextureSample(UVs=Multiply(A=@n5, B=@n10)) {texture=/Game/Global/Textures/Default/Caustics/T_Turbulence_Seamless}.G, B=@n4.R)
  NamedRerouteDeclaration(None=Multiply(A=Multiply(A=@n7) {constB=10}, B=Multiply(A=@n9, B=@n11))) {name=Shape, nodeColor=<rgba 0.2891 1 0 1>}
  n12 = Subtract(A=ParticlePositionWS(), B=WorldPosition().XYZ)
  n13 = Constant()
  n14 = ScalarParameter() {defaultValue=166, parameterName=ScaleDown_Start}
  n15 = Clamp(None=Subtract(
    A=MaterialFunctionCall(Distance=@n14) {materialFunction=&Near_Camera_Fade}.Result
    B=ScalarParameter() {parameterName=ScaleDown_End_Normalized}
   ))
  NamedRerouteDeclaration(None=ShadingPathSwitch(Default=LinearInterpolate(A=@n12, B=@n13, Alpha=@n15), Mobile=Constant()) {desc=Migrated from FeatureLevelSwitch}) {name=WPO_Scale, nodeColor=<rgba 0 0.7891 1 1>}
  NamedRerouteDeclaration(None=ShadingPathSwitch(Default=@n15, Mobile=Constant() {r=1}) {desc=Migrated from FeatureLevelSwitch}) {name=Camera_Fade, nodeColor=<rgba 1 0.8906 0 1>}
  NamedRerouteUsage()
  n16 = ParticleRandom()
  n17 = NamedRerouteUsage()
  n18 = Multiply(
   A=Multiply(
    A=@n1.A
    B=MaterialFunctionCall!(Random Picker=@n16, Texture Input=@n17, Option 1, Option 2, Option 3) {materialFunction=/Game/Effects/Fort_Effects/Materials/_Functions/MF_RGB_Random_01}.Result
   )
   B=NamedRerouteUsage()
  )
  StaticSwitchParameter(True=Multiply(A=DistanceCullFade(), B=@n18), False=@n18) {parameterName=UseDistanceCullFade}
 ]
}
