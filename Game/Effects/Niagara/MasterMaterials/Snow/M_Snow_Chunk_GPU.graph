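# Translucent master material for GPU snow chunk sprites in Niagara and Cascade, built on soft smoke color and normal textures with random 2x2 sub-UV frames.
# Time-of-day tint, from a curve atlas or four-color function, multiplies particle color for base color and emissive, with unlit and texture switches.
# Opacity morphs from sphere mask to texture alpha as faded particle alpha rises, using near-camera and non-mobile depth fades.
# Optional world position offset shrinks sprites by camera distance.
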Material /Game/Effects/Niagara/MasterMaterials/Snow/M_Snow_Chunk_GPU
graph {
 properties {
  bUsedWithNiagaraSprites=true
  bUsedWithParticleSprites=true
  blendMode=BLEND_Translucent
  pixelDepthOffsetMode=PDOM_Legacy
  translucencyLightingMode=TLM_VolumetricPerVertexDirectional
  usedShadingModels=Default Lit
 }
 nodes [
  MaterialFunctionCall(Emissive=NamedRerouteUsage()) {materialFunction=&MF_FX_EmissiveAdjustments}
  n1 = CurveAtlasRowParameter(CurveTime=CollectionParameter() {collection=&FortniteMaterialParameters, parameterName=TimeOfDay_0-1}) {
   atlas=&AthenaCurveAtlas01
   curve=/Game/Athena/LinearColorCurves/Curve_TimeOfDay_S23
   defaultValue=321
   group=TimeOfDay
   parameterName=TOD_Curve
  }
  n2 = MaterialFunctionCall(
   Color_Morning=VectorParameter() {parameterName=Color Morning, defaultValue=<rgba 1 0.3343 0.122 1>}.RGB
   Color_Day=VectorParameter() {parameterName=Color Day, defaultValue=<rgba 1 0.9054 0.7298 1>}.RGB
   Color_Evening=VectorParameter() {parameterName=Color Evening, defaultValue=<rgba 1 0.1673 0.1291 1>}.RGB
   Color_Night=VectorParameter() {parameterName=Color Night, defaultValue=<rgba 0.229 0.442 1 1>}.RGB
  ) {materialFunction=&Color_Time_of_Day}
  n3 = ParticleColor()
  n4 = Multiply(A=StaticSwitchParameter(True=@n1, False=LinearInterpolate(B=@n2.Result) {constA=1}) {parameterName=Use ToD Color Curve}, B=@n3.RGB)
  n5 = TextureSample(UVs=NamedRerouteUsage()) {samplerType=SAMPLERTYPE_LinearColor, texture=/Game/Effects/Fort_Effects/Textures/Smoke/SmokeTest2_Soft}
  n6 = Multiply(A=@n4, B=@n5.R)
  n7 = Multiply(
   A=@n4
   B=Add(
    A=ScalarParameter() {defaultValue=0.05, parameterName=Emissive Strength}
    B=DynamicParameter() {defaultValue=<rgba 0 0 1 1>, paramNames [EmissiveStrength, SofterFlakes, Param3, Param4]}.EmissiveStrength
   )
  ) {constB=0.05}
  NamedRerouteDeclaration(None=StaticSwitchParameter(True=StaticSwitchParameter(True=@n6, False=@n4) {parameterName=UseTexture}, False=Multiply(A=@n7, B=@n5.R)) {parameterName=Is Unlit}) {name=EmissiveColor, nodeColor=<rgba 0.4062 0 1 1>}
  NamedRerouteDeclaration(None=@n4) {name=BaseColor, nodeColor=<rgba 0 1 0.0859 1>}
  n8 = Subtract(A=ParticlePositionWS(), B=WorldPosition().XYZ)
  n9 = MaterialFunctionCall(
   Fade Length=ScalarParameter() {defaultValue=2000, group=Shrink at distance, parameterName=Shrink Distance}
   Fade Offset=ScalarParameter() {defaultValue=200, group=Shrink at distance, parameterName=Shrink Distance Offset}
   For Vertex Shader=StaticBool() {value=true}
  ) {materialFunction=&CameraDepthFade}
  NamedRerouteDeclaration(None=StaticSwitchParameter(True=Multiply(A=@n8, B=@n9.Result), False=Constant()) {group=Shrink at distance, parameterName=Shrink Particles At distance}) {name=WPO, nodeColor=<rgba 1 0 0.2344 1>}
  n10 = TextureSample(UVs=NamedRerouteUsage()) {samplerType=SAMPLERTYPE_Normal, texture=/Game/Effects/Fort_Effects/Textures/Smoke/SmokeTest_N}
  n11 = Constant3Vector() {constant=<rgba 0.5 0.5 1 0>}
  n12 = Add(
   A=ScalarParameter() {parameterName=Use Softer Flakes}
   B=DynamicParameter() {defaultValue=<rgba 0 0 1 1>, paramNames [EmissiveStrength, SofterFlakes, Param3, Param4]}.SofterFlakes
  )
  NamedRerouteDeclaration(None=LinearInterpolate(A=@n10.RGB, B=Multiply(A=@n10.RGB, B=@n11), Alpha=@n12)) {name=Normal, nodeColor=<rgba 0 0.5547 1 1>}
  n13 = MaterialFunctionCall(Distance=ScalarParameter() {defaultValue=150, parameterName=Near Camera Fade Distance}) {materialFunction=&Near_Camera_Fade}
  n14 = ScalarParameter() {defaultValue=55, parameterName=DepthFadeDist}
  n15 = Multiply(
   A=@n3.A
   B=ShadingPathSwitch(Default=DepthFade(Opacity=@n13.Result, FadeDistance=@n14), Mobile=Constant() {r=1}) {desc=Migrated from FeatureLevelSwitch}
  )
  n16 = Constant() {r=0.5}
  n17 = SphereMask(A=TextureCoordinate(), B=@n16, Radius=@n16) {hardnessPercent=0}
  n18 = LinearInterpolate(A=@n5.A, B=@n5.A, Alpha=@n12)
  n19 = Power(Base=@n15) {constExponent=4}
  NamedRerouteDeclaration(None=Multiply(A=@n15, B=LinearInterpolate(A=@n17, B=@n18, Alpha=@n19))) {name=Opacity, nodeColor=<rgba 0.875 1 0 1>}
  n20 = Constant() {r=2}
  NamedRerouteDeclaration(None=MaterialFunctionCall!(UseTexCoord 1, Rows=@n20, Columns=@n20, Particle Random Value=ParticleRandom()) {materialFunction=/Game/Effects/Fort_Effects/Materials/_Functions/MF_RandomSubUV_GPU_Sprites}.Result) {name=ParticleRandom, nodeColor=<rgba 0.8047 0 1 1>}
  NamedRerouteUsage()
  NamedRerouteUsage()
  NamedRerouteUsage()
  NamedRerouteUsage()
 ]
}
